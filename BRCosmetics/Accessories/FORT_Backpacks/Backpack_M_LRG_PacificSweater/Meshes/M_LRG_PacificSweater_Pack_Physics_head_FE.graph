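# Physics asset for the PacificSweater large back bling, a bear-pack cosmetic, giving its dangling parts limited secondary motion.
# Single capsule bodies cover the master_bear_pack root, arms, legs, head, ears, three hair bones and horn.
# Limbs and head constrain to the root, while ears, hair and horn constrain to the head.
# Twist is nearly locked throughout; swing is tight on head and horn, looser on ears and limbs, widest on hair.
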PhysicsAsset /BRCosmetics/Accessories/FORT_Backpacks/Backpack_M_LRG_PacificSweater/Meshes/M_LRG_PacificSweater_Pack_Physics_head_FE
bodies {
 dyn_arm_l [{shapeType=Capsule, length=3, radius=4, shapeName=dyn_arm_l_capsule, center=<4 0 1>, rotation=<rot 90 0 0>}]
 dyn_arm_r [{shapeType=Capsule, length=3, radius=4, shapeName=dyn_arm_l_capsule, center=<-4 0 -1>, rotation=<rot 90 0 0>}]
 dyn_ear_l [{shapeType=Capsule, length=2, radius=2, shapeName=dyn_head_capsule, center=<3 0 0>, rotation=<rot 90 0 0>}]
 dyn_ear_r [{shapeType=Capsule, length=2, radius=2, shapeName=dyn_head_capsule, center=<-3 0 0>, rotation=<rot 90 0 0>}]
 dyn_hair_01 [{shapeType=Capsule, length=3, radius=2, shapeName=dyn_hair_01_capsule, center=<4 0 0>, rotation=<rot 90 0 0>}]
 dyn_hair_02 [{shapeType=Capsule, length=3, radius=2, shapeName=dyn_hair_01_capsule, center=<4 0 0>, rotation=<rot 90 0 0>}]
 dyn_hair_03 [{shapeType=Capsule, length=3, radius=2, shapeName=dyn_hair_01_capsule, center=<4 0 0>, rotation=<rot 90 0 0>}]
 dyn_head [{shapeType=Capsule, length=5, radius=4, shapeName=dyn_head_capsule, center=<3 -3 0>, rotation=<rot 90 0 90>}]
 dyn_horn [{shapeType=Capsule, length=5, radius=2, shapeName=dyn_horn_capsule, center=<3 0 0>, rotation=<rot 90 0 0>}]
 dyn_leg_l [{shapeType=Capsule, length=3, radius=4, shapeName=dyn_arm_l_capsule, center=<5 0 0>, rotation=<rot 90 0 0>}]
 dyn_leg_r [{shapeType=Capsule, length=3, radius=4, shapeName=dyn_arm_l_capsule, center=<-5 0 0>, rotation=<rot 90 0 0>}]
 master_bear_pack [{shapeType=Capsule, length=2, radius=3, shapeName=master_bear_pack_capsule}]
}
constraints [
 {bone1Name=dyn_arm_l, bone2Name=master_bear_pack, swing1LimitDegrees=25, swing2LimitDegrees=25, twistLimitDegrees=1}
 {bone1Name=dyn_arm_r, bone2Name=master_bear_pack, swing1LimitDegrees=25, swing2LimitDegrees=25, twistLimitDegrees=1}
 {bone1Name=dyn_leg_l, bone2Name=master_bear_pack, swing1LimitDegrees=25, swing2LimitDegrees=25, twistLimitDegrees=1}
 {bone1Name=dyn_leg_r, bone2Name=master_bear_pack, swing1LimitDegrees=25, swing2LimitDegrees=25, twistLimitDegrees=1}
 {bone1Name=dyn_head, bone2Name=master_bear_pack, swing1LimitDegrees=2, swing2LimitDegrees=5, twistLimitDegrees=1}
 {bone1Name=dyn_ear_r, bone2Name=dyn_head, swing1LimitDegrees=15, swing2LimitDegrees=15, twistLimitDegrees=1}
 {bone1Name=dyn_ear_l, bone2Name=dyn_head, swing1LimitDegrees=15, swing2LimitDegrees=15, twistLimitDegrees=1}
 {bone1Name=dyn_hair_01, bone2Name=dyn_head, swing1LimitDegrees=35, swing2LimitDegrees=20, twistLimitDegrees=1}
 {bone1Name=dyn_hair_02, bone2Name=dyn_head, swing1LimitDegrees=40, swing2LimitDegrees=20, twistLimitDegrees=1}
 {bone1Name=dyn_hair_03, bone2Name=dyn_head, swing1LimitDegrees=35, swing2LimitDegrees=20, twistLimitDegrees=1}
 {bone1Name=dyn_horn, bone2Name=dyn_head, swing1LimitDegrees=5, swing2LimitDegrees=5, twistLimitDegrees=1}
]
